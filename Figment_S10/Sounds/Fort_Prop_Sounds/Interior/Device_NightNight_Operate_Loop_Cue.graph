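# Sound cue for the operating loop of the NightNight interior prop device.
# A single looping wave player feeds the root directly, playing the Device_NightNight_Operate_Loop wave at reduced volume in the Fort_Gadgets sound class.
# It references the TV prop attenuation asset but overrides it with its own logarithmic falloff settings.
# A concurrency override uses the PreventNew rule, so new instances are blocked once the limit is reached.
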SoundCue /Figment_S10/Sounds/Fort_Prop_Sounds/Interior/Device_NightNight_Operate_Loop_Cue
root=WavePlayer_3
nodes [{id=WavePlayer_3, properties {bLooping=true, soundWaveAssetPtr=&/Device_NightNight_Operate_Loop}}]
properties {
 attenuationSettings=/Game/Sounds/Attenuation/Fort_Prop_TV_Att
 bOverrideAttenuation=true
 bOverrideConcurrency=true
 maxDistance=1400
 soundClassObject=&Fort_Gadgets
 volumeMultiplier=0.3
 attenuationOverrides {distanceAlgorithm=Logarithmic, falloffDistance=1000}
 concurrencyOverrides {resolutionRule=PreventNew}
}
